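# Masked subsurface material for the Gumball skydive contrail cosmetic, built as a multi-option effect shader driven by Niagara dynamic parameters.
# World position offset combines masked sine wobble, near-camera damping, bending toward a Niagara-supplied value and optional SplineThicken "NitroFlow" ribbons.
# Color blends four-color gradients, gumball SubUV textures or particle color, alongside a fresnel and fake-reflection "Slurp" look.
# Switchable scalar lerps also select panning or SubUV normals and erosion or dot opacity.
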Material /BRCosmetics/Athena/Effects/Particles/Skydiving/NS_Contrail_Gumball/M_Gumball_SkyDive
graph {
 properties {
  bUsedWithMeshDeformer=true
  bUsedWithNanite=true
  bUsedWithNiagaraMeshParticles=true
  bUsedWithSkeletalMesh=true
  blendMode=BLEND_Masked
  pixelDepthOffsetMode=PDOM_Legacy
  shadingModel=MSM_Subsurface
  usedShadingModels=Subsurface
 }
 nodes [
  n1 = DynamicParameter() {desc=Parameters set by Niagara, parameterIndex=2, paramNames [X,Y,Z,WPOIntensity]}
  n2 = Subtract(
   A=MaterialFunctionCall(X=@n1.X, Y=@n1.X, Z=@n1.X) {materialFunction=&MakeFloat3}.Result
   B=ParticlePositionWS()
  )
  n3 = ScalarParameter() {defaultValue=-1000, group=WPO, parameterName=InterpolateWPO_Clamp_min}
  n4 = ScalarParameter() {defaultValue=1000, group=WPO, parameterName=InterpolateWPO_Clamp_max}
  n5 = Power(Base=OneMinus(None=VertexColor().R), Exp=ScalarParameter() {defaultValue=3, group=WPO, parameterName=InterpolateWPO_BendPower})
  n6 = Multiply(
   A=TextureCoordinate() {uTiling=0, vTiling=0.7}
   B=ComponentMask(None=VectorParameter() {group=WPO, parameterName=WPO_UV_Scale, defaultValue=<rgba 1 1 0 1>}.RGB) {g=true, r=true}
  )
  n7 = Add(A=Add(A=ComponentMask(None=@n6) {g=true}, B=ScalarParameter() {group=WPO, parameterName=WPO_Offset}), B=VertexColor().G)
  n8 = DynamicParameter() {defaultValue=<rgba 1 1 1 0>, paramNames [RandomOffset, WPO_Intensity, WPO_Speed, Dissolve]}
  n9 = Add(
   A=Add(A=Add(A=@n7, B=@n8.RandomOffset), B=VertexColor().R)
   B=Multiply(A=Time(), B=DynamicParameter() {defaultValue=<rgba 1 1 1 0>, paramNames [RandomOffset, WPO_Intensity, WPO_Speed, Dissolve]}.WPO_Speed) {constB=2.8}
  ) {constB=0}
  n10 = Multiply(
   A=VectorParameter() {group=WPO, parameterName=WPO_Direction, defaultValue=<rgba 0.25 0.25 0 1>}.RGB
   B=Multiply(A=Sine(None=@n9)) {constB=20}
  )
  n11 = OneMinus(None=ComponentMask(None=TextureCoordinate()) {g=true})
  n12 = Saturate(None=LinearInterpolate(Alpha=Add(A=@n11) {constB=0.5}))
  n13 = Multiply(A=Transform(None=Multiply(A=@n10, B=@n12)) {transformSourceType=TRANSFORMSOURCE_View}) {constB=0.5}
  n14 = DynamicParameter() {defaultValue=<rgba 1 1 1 0>, paramNames [RandomOffset, WPO_Intensity, WPO_Speed, Dissolve]}
  n15 = OneMinus(None=ComponentMask(None=TextureCoordinate()) {g=true})
  n16 = ScalarParameter() {group=WPO, parameterName=Bottom_Mask_Offset}
  n17 = Multiply(
   A=Power(Base=Add(A=@n15, B=@n16), Exp=ScalarParameter() {defaultValue=2.9, group=WPO, parameterName=Power})
   B=ScalarParameter() {defaultValue=50, group=WPO, parameterName=Power_Mask_Edge}
  ) {constB=50}
  n18 = LinearInterpolate(
   A=ScalarParameter() {defaultValue=0.3, group=WPO, parameterName=Min_Near_Camera_WPO}
   Alpha=MaterialFunctionCall(Distance=ScalarParameter() {defaultValue=256, group=WPO, parameterName=Near_Camera_fade}) {materialFunction=&Near_Camera_Fade}.Result
  ) {constA=0.5}
  n19 = Multiply(
   A=Multiply(A=Multiply(A=@n13, B=@n14.WPO_Intensity), B=Saturate(None=@n17))
   B=LinearInterpolate(B=Saturate(None=@n18), Alpha=ScalarParameter() {group=WPO, parameterName=Use_Near_Camera_Fade}) {constA=1}
  )
  n20 = ScalarParameter() {defaultValue=1, group=WPO, parameterName=Use_WPO}
  n21 = ScalarParameter() {group=WPO, parameterName=Small_WPO_Intensity}
  n22 = NamedRerouteUsage()
  n23 = Add(
   A=LinearInterpolate(B=Clamp(None=@n2, Min=@n3, Max=@n4) {maxDefault=1000, minDefault=-1000}, Alpha=Multiply(A=@n5, B=@n1.WPOIntensity))
   B=Add(A=LinearInterpolate(B=@n19, Alpha=@n20), B=Multiply(A=@n21, B=@n22)) {constB=0}
  )
  n24 = MaterialFunctionCall(
   WidthBase=ScalarParameter() {defaultValue=10, group=NitroFlow_Options, parameterName=BaseWidth, sortPriority=0}
   WidthTip=ScalarParameter() {defaultValue=1, group=NitroFlow_Options, parameterName=TipWidth, sortPriority=1}
  ) {materialFunction=&SplineThicken}
  n25 = ComponentMask(None=LinearInterpolate(
    A=TextureCoordinate()
    B=TextureCoordinate() {coordinateIndex=1}
    Alpha=ScalarParameter() {defaultValue=1, group=NitroFlow_Options, parameterName=NitroFlow_UseUV1}
   )) {g=true}
  n26 = Power(
   Base=Multiply(A=@n25) {constB=2}
   Exp=ScalarParameter() {defaultValue=1, group=NitroFlow_Options, parameterName=Tip_GradientPower, sortPriority=3}
  ) {constExponent=1}
  n27 = ScalarParameter() {defaultValue=1, group=NitroFlow_Options, parameterName=Tip_GradientMultiply, sortPriority=2}
  n28 = Power(
   Base=Multiply(A=OneMinus(None=@n25)) {constB=2}
   Exp=ScalarParameter() {defaultValue=1, group=NitroFlow_Options, parameterName=Base_GradientPower, sortPriority=4}
  ) {constExponent=1}
  n29 = ScalarParameter() {defaultValue=1, group=NitroFlow_Options, parameterName=Base_GradientMultiply, sortPriority=5}
  n30 = Multiply(A=Saturate(None=Multiply(A=@n26, B=@n27)), B=Saturate(None=Multiply(A=@n28, B=@n29)))
  Add(
   A=@n23
   B=LinearInterpolate(
    A=@n23
    B=Multiply(A=@n24.WorldPositionOffset, B=@n30)
    Alpha=ScalarParameter() {group=NitroFlow_Options, parameterName=UseNitroFlow_SplineThicken, sortPriority=-1}
   )
  )
  ComponentMask(None=OneMinus(None=TextureCoordinate())) {g=true}
  Constant() {r=1}
  n31 = DynamicParameter() {defaultValue=<rgba 1 1 1 0>, paramNames [RandomOffset, WPO_Intensity, WPO_Speed, Dissolve]}
  n32 = AppendVector(A=@n31.RandomOffset, B=@n31.RandomOffset)
  n33 = DynamicParameter() {parameterIndex=1, paramNames [Erosion_U, Erosion_V, Brightness_in_game, Use_Frontend_bpr]}
  n34 = Multiply(A=TextureCoordinate() {uTiling=0.85, vTiling=0.85}, B=AppendVector(A=@n33.Erosion_U, B=@n33.Erosion_U))
  n35 = TextureSampleParameter2D(UVs=NamedRerouteDeclaration(None=Add(A=@n32, B=@n34)) {name=ErosionUVs, nodeColor=<rgba 1 1 1 1>}) {
   group=Mask
   parameterName=Mask
   samplerType=SAMPLERTYPE_LinearColor
   texture=/Game/Athena/Apollo/Water/Textures/02_Rounded/02_WaterFoam_Veroni_base_H
  }
  n36 = LinearInterpolate(A=@n35.RGB, B=OneMinus(None=@n35.RGB), Alpha=ScalarParameter() {parameterName=InverseErocionMask})
  n37 = DynamicParameter() {defaultValue=<rgba 1 1 1 0>, paramNames [RandomOffset, WPO_Intensity, WPO_Speed, Dissolve]}
  n38 = TextureCoordinate()
  n39 = Constant() {r=0.5}
  n40 = ScalarParameter() {defaultValue=0.5, group=NitroFlow_Options, parameterName=NitroFlow_DotRadius}
  n41 = ScalarParameter() {group=NitroFlow_Options, parameterName=NitroFlow_DotHardness}
  LinearInterpolate(
   A=LinearInterpolate(A=Constant() {r=1}, B=Add(A=@n36, B=@n37.Dissolve) {constB=10}, Alpha=ScalarParameter() {group=Mask, parameterName=Use_Mask_Erocion})
   B=Multiply(A=ParticleColor().A, B=SphereMask(A=@n38, B=@n39, Radius=@n40, Hardness=@n41) {hardnessPercent=0})
   Alpha=ScalarParameter() {group=NitroFlow_Options, parameterName=UseNitroFlow_DotsOpacity, sortPriority=-1}
  )
  n42 = VectorParameter() {parameterName=Normal UV, defaultValue=<rgba 2 2 0 0>}
  n43 = AppendVector(A=Constant(), B=VertexColor().R)
  n44 = Add(A=Multiply(A=TextureCoordinate() {uTiling=1.5, vTiling=1.5}, B=ComponentMask(None=@n42.RGB) {g=true, r=true}), B=@n43)
  n45 = VectorParameter() {parameterName=Normal Pan Speed, defaultValue=<rgba 0 0.1 0 1>}
  n46 = NamedRerouteDeclaration(None=TextureSample(UVs=Panner(Coordinate=@n44, Speed=@n45.RGB) {speedY=2}) {samplerType=SAMPLERTYPE_Normal, texture=/BRCosmetics/Characters/Player/Male/Large/Bodies/M_LRG_Clash_V/Textures/Simple_Perlin_N}.RGB) {name=Normal, nodeColor=<rgba 1 0.4922 0 1>}
  n47 = Add(A=@n43, B=TextureCoordinate() {uTiling=2, vTiling=2})
  n48 = LinearInterpolate(Alpha=TextureSample(UVs=Panner(Coordinate=@n47) {speedY=-0.3}) {texture=/Game/Effects/Fort_Effects/Textures/Power_Source/T_Power_Source_Sphere_Texture}.R) {constA=-0.2, constB=5}
  n49 = AppendVector(A=Multiply(A=ComponentMask(None=@n46) {g=true, r=true}, B=Saturate(None=@n48)), B=Constant() {r=1})
  n50 = ScalarParameter() {defaultValue=0.9, group=Normal, parameterName=Flatness}
  n51 = TextureCoordinate()
  n52 = Multiply(A=Power(Base=ComponentMask(None=@n51) {g=true})) {constB=10}
  n53 = TextureSampleParameter2D(UVs=NamedRerouteUsage()) {
   group=Normal
   parameterName=Normal
   samplerType=SAMPLERTYPE_Normal
   texture=/Game/Athena/Apollo/Water/Textures/02_Rounded/02_WaterFoam_Veroni_base_N
  }
  n54 = TextureSampleParameterSubUV!(UVs, Apply View MipBias) {parameterName=Normal_SubUV, samplerType=SAMPLERTYPE_Normal, texture=/Game/Characters/Player/Male/Medium/Bodies/M_MED_Gumball/FX/T_Gumballs_N}
  LinearInterpolate(
   A=LinearInterpolate(
    A=Constant3Vector() {constant=<rgba 0 0 1 0>}
    B=MaterialFunctionCall(Normal=@n49, Flatness=@n50) {materialFunction=&FlattenNormal}.Result
    Alpha=Saturate(None=@n52)
   )
   B=LinearInterpolate(A=@n53.RGB, B=@n54.RGB, Alpha=ScalarParameter() {defaultValue=1, parameterName=Use_Sub_UV})
   Alpha=ScalarParameter() {group=Normal, parameterName=Use_Normal_Texture}
  )
  NamedRerouteDeclaration(None=ScalarParameter() {defaultValue=0.5, parameterName=Specular}) {name=Specular, nodeColor=<rgba 0 0.2031 1 1>}
  NamedRerouteDeclaration(None=ScalarParameter() {parameterName=Metalic}) {name=Metalic, nodeColor=<rgba 0 0.8594 1 1>}
  NamedRerouteDeclaration(None=ScalarParameter() {defaultValue=0.35, parameterName=Roughness}) {name=Roughness, nodeColor=<rgba 0.2891 0 1 1>}
  n55 = VectorParameter() {group=Diffuse, parameterName=Color1, defaultValue=<rgba 0.0923 0 1 0>}
  n56 = VectorParameter() {group=Diffuse, parameterName=Color2, defaultValue=<rgba 0.64 0 1 0>}
  n57 = ComponentMask(None=TextureCoordinate() {uTiling=1.5, vTiling=1.5}) {g=true}
  n58 = MaterialFunctionCall(In=Add(A=@n57) {constB=0}, Contrast=Constant() {r=0.5}) {materialFunction=&CheapContrast}
  n59 = VectorParameter() {group=Diffuse, parameterName=Color3, defaultValue=<rgba 0.0923 0 1 0>}
  n60 = VectorParameter() {group=Diffuse, parameterName=Color4, defaultValue=<rgba 0 0.4286 1 0>}
  n61 = MaterialFunctionCall(In=Add(A=@n57) {constB=-0.8}, Contrast=Constant() {r=0.5}) {materialFunction=&CheapContrast}
  n62 = Add(A=@n57) {constB=-0.4}
  n63 = Constant() {r=0.5}
  n64 = TextureSampleParameterSubUV() {
   group=Diffuse
   parameterName=Diffuse_SubUV
   samplerSource=SSM_Wrap_WorldGroupSettings
   sortPriority=2
   texture=/Game/Characters/Player/Male/Medium/Bodies/M_MED_Gumball/FX/T_Gumballs_D
  }
  n65 = LinearInterpolate(
   A=LinearInterpolate(
    A=LinearInterpolate(A=@n55.RGB, B=@n56.RGB, Alpha=@n58.Result)
    B=LinearInterpolate(A=@n59.RGB, B=@n60.RGB, Alpha=@n61.Result)
    Alpha=MaterialFunctionCall(In=@n62, Contrast=@n63) {materialFunction=&CheapContrast}.Result
   )
   B=@n64.RGB
   Alpha=ScalarParameter() {group=Diffuse, parameterName=Use_Color_Texture}
  )
  n66 = ParticleColor()
  n67 = MaterialFunctionCall(R=@n66.R, G=@n66.R, B=@n66.R, A=@n66.R) {materialFunction=&AppendMany}
  n68 = ScalarParameter() {group=Diffuse, parameterName=Use_Particle_Color}
  n69 = NamedRerouteDeclaration(None=LinearInterpolate(
    A=LinearInterpolate(A=@n65, B=@n67.RGB, Alpha=@n68)
    B=VectorParameter() {group=Diffuse, parameterName=SubSurfaceColor, defaultValue=<rgba 0.0923 0 1 0>}.RGB
    Alpha=ScalarParameter() {parameterName=Use Simple SS Color}
   )) {name=Color, nodeColor=<rgba 1 0 0.0703 1>}
  n70 = DynamicParameter() {parameterIndex=1, paramNames [Erosion_U, Erosion_V, Brightness_in_game, Use_Frontend_bpr]}
  n71 = LinearInterpolate(
   A=Multiply(A=Multiply(A=@n69, B=@n70.Brightness_in_game) {constB=0.05})
   B=NamedRerouteUsage()
   Alpha=ScalarParameter() {parameterName=UseSlurpColors}
  )
  n72 = ParticleColor()
  n73 = ScalarParameter() {group=NitroFlow_Options, parameterName=UseNitroFlow_Color, sortPriority=-1}
  NamedRerouteDeclaration(None=MaterialFunctionCall(Emissive=LinearInterpolate(A=@n71, B=@n72.RGB, Alpha=@n73)) {materialFunction=&MF_FX_EmissiveAdjustments}.Result) {name=Emissive, nodeColor=<rgba 1 0.3984 0 1>}
  n74 = MaterialFunctionCall(
   Brightness=ScalarParameter() {defaultValue=1.3, parameterName=FresnelMult}
   Exponent=ScalarParameter() {defaultValue=2, parameterName=FresnelExp}
  ) {materialFunction=&MF_Fresnel_FX}
  n75 = LinearInterpolate(
   A=VectorParameter() {parameterName=ColorA, defaultValue=<rgba 0 0.1875 0.1037 1>}.RGB
   B=VectorParameter() {parameterName=ColorB, defaultValue=<rgba 0.0335 0.1033 0.1927 1>}.RGB
   Alpha=@n74.Fresnel
  )
  n76 = Multiply(A=@n75, B=@n74.Fresnel Inverted)
  n77 = TextureCoordinate()
  n78 = ScalarParameter() {defaultValue=0.6, parameterName=DistTile}
  n79 = ScalarParameter() {defaultValue=2, parameterName=DistSpeed}
  n80 = Constant() {r=0.1}
  n81 = TextureSample(UVs=Panner(Coordinate=Multiply(A=@n77, B=@n78), Time=Time() {bOverride_Period=true, period=64}, Speed=AppendVector(A=@n79, B=@n80)) {speedX=-1, speedY=0.1}) {samplerType=SAMPLERTYPE_Normal, texture=&T_Weap_Distort_N_02}
  n82 = Add(A=Multiply(A=Transform(None=@n81.RGB), B=ScalarParameter() {defaultValue=0.5, parameterName=NoiseSTr}), B=VertexNormalWS())
  n83 = Add(A=Multiply(A=Transform(None=@n82) {transformSourceType=TRANSFORMSOURCE_World, transformType=TRANSFORM_View}) {constB=0.5}) {constB=0.5}
  n84 = MaterialFunctionCall(
   In=TextureSample(UVs=ComponentMask(None=@n83) {g=true, r=true}) {texture=/Game/Packages/Fortress_SharedMaterials/Textures/T_Ravens_SpecularReflection}.RGB
   Contrast=ScalarParameter() {parameterName=ReflectionContrast}
  ) {materialFunction=&CheapContrast}
  n85 = VectorParameter() {parameterName=ReflectionColor, defaultValue=<rgba 0 0.3 4 1>}
  NamedRerouteDeclaration(None=Add(A=Add(A=@n75, B=@n76), B=Multiply(A=@n84.Result, B=@n85.RGB))) {name=FresnelColor, nodeColor=<rgba 0.9688 1 0 1>}
  NamedRerouteUsage()
  NamedRerouteUsage()
  NamedRerouteUsage()
  NamedRerouteUsage()
  NamedRerouteUsage()
  NamedRerouteUsage()
  VectorParameter() {parameterName=vdscv}
 ]
}
