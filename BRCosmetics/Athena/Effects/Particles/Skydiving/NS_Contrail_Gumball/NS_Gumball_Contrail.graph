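# Niagara skydiving contrail for the Gumball cosmetic, emitting mesh particles from the player's hands.
# Two BubbleGuys emitters continuously spawn randomized meshes at hand bones, which drift with curl noise, partly follow source movement and shrink away.
# Left_hand_tendrils holds three bone-pinned meshes that stretch with owner speed and write smoothed positions to a dynamic vector.
# System Dive blends offsets and material values, while a front-end preview flag adds bubble velocity and extra stretch.
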NiagaraSystem /BRCosmetics/Athena/Effects/Particles/Skydiving/NS_Contrail_Gumball/NS_Gumball_Contrail
dependencies {
 usedDataInterfaces [
  &NiagaraDataInterfaceColorCurve
  &NiagaraDataInterfaceCurve
  &NiagaraDataInterfaceCurlNoise
  &NiagaraDataInterfaceArrayFloat
  &NiagaraDataInterfaceArrayFloat3
 ]
 usedDynamicInputs [&FN_FrontEndFloatModifier, &SetBoolByFloatComparison]
 usedModules [
  /Game/Effects/Niagara/Modules/Skydiving/FN_DivingAttributes/v4/FN_DivingAttributes_v4
  &SystemState
  &self:SystemSpawnScript.SystemScriptSource.SystemScriptGraph.NiagaraNodeAssignment_0.SetVariables_#1
  &FN_Integrator
 ]
}
emitters {
 BubbleGuys {
  emitterUpdateScript [
   &EmitterState()
   &SpawnRate(SpawnRate=6, Use Spawn Probability=false)
  ]
  particleSpawnScript [
   &FN_SetSpawnGroup(Number of Spawn Groups=2) {module=FN_SetSpawnGroup001}
   &InitializeParticle(
    Color Mode=::NewEnumerator0
    Lifetime Max=1.35
    Lifetime Min=0.75
    Lifetime Mode=::NewEnumerator1
    Mesh Scale Mode=::NewEnumerator1
    Mesh Uniform Scale Max=3.75
    Mesh Uniform Scale Min=2.5
    UsePositionOffset=false
    Write Material Random=true
   ) {module=InitializeParticle001}
   &SkeletalMeshLocation(
    Bone / Socket Index=$Particles.SpawnGroup
    Bone Sampling Mode=::NewEnumerator4
    Inherit Velocity=false
    MaskSpawnGroup=false
    Sampled Position Offset=&RandomRangeVector(Minimum=<-3 -3 -3>, Maximum=<3 3 3>)
    Skeletal Mesh=$User.Hands
   )
   &InitialMeshOrientation(
    Orientation Axis=<0.5 0 0.5>
    Rotation=&RandomRangeVector(Minimum=<0 -0.1 -1>, Maximum=<0 0.1 1>)
    Use Orientation Vector=true
    Use Rotation Vector=true
   )
   &SubUVAnimation(End Frame=3, SubUV Animation Mode:ENiagara_SubUVLookupMode=::NewEnumerator2)
   &AddVelocity(Velocity=&MakeCustomVectorFromBool(
     Bool=$User.bIsFrontEndPreview
     True Vector=&Lerp_Vector(A=<0 0 150>, B=<0 0 400>, Alpha=$System.Dive)
     False Vector=<0 0 0>
    ))
  ]
  particleUpdateScript [
   &ParticleState()
   /SetVariables(Particles.EngineOwnerForwardAxis:Vector3f=$Engine.Owner.SystemXAxis) {enabled=false, isSetParameters=true}
   /SetVariables(Particles.Lean:NiagaraFloat=&RemapRange(
     Input Value To Remap=HlslExpression() {hlslExpression="dot((normalize(Engine.Owner.Velocity)), Particles.EngineOwnerForwardAxis)"}
     Input Min=-0.5
     Input Max=0.5
     Output Min=0
     Output Max=1
    )) {enabled=false, isSetParameters=true}
   /SetVariables(Particles.Color=&ColorFromCurve(
     ColorCurve=DataInterface() {propertyValues {
       AlphaCurve {keys [<at 0 1 RCIM_Cubic>]}
       BlueCurve {keys [
         {arriveTangent=0.1276, interpMode=RCIM_Cubic, leaveTangent=0.1276, time=0, value=0}
         <at 0.5 1>
         {arriveTangent=0.1276, interpMode=RCIM_Cubic, leaveTangent=0.1276, time=1, value=0.1276}
        ]}
       GreenCurve {keys [
         {arriveTangent=0.9735, interpMode=RCIM_Cubic, leaveTangent=0.9735, time=0, value=0.0265}
         <at 0.5 1>
         {arriveTangent=0.9735, interpMode=RCIM_Cubic, leaveTangent=0.9735, time=1, value=1}
        ]}
       RedCurve {keys [
         {arriveTangent=-1, interpMode=RCIM_Cubic, leaveTangent=-1, time=0, value=1}
         <at 0.5 1>
         {arriveTangent=-1, interpMode=RCIM_Cubic, leaveTangent=-1, time=1, value=0}
        ]}
      }}
     CurveIndex=$Particles.Lean
    )) {enabled=false, isSetParameters=true}
   &InheritSourceMovement(
    Applied Force Scale=&VectorFromFloat(Value=&FloatFromCurve(FloatCurve=DataInterface() {propertyValues {Curve {keys [<at 0 0.7>,<at 1 0>]}}}))
    Applied Position Falloff Distance=700
    Applied Position Scale=&VectorFromFloat(Value=&FloatFromCurve(FloatCurve=DataInterface() {propertyValues {Curve {keys [
          {arriveTangent=-0.2939, interpMode=RCIM_Cubic, leaveTangent=-0.2939, tangentMode=RCTM_User, time=0, value=1}
          {arriveTangent=-1.2984, interpMode=RCIM_Cubic, leaveTangent=-1.2984, tangentMode=RCTM_User, time=1, value=0.5}
         ]}}}))
    ApplyForce=false
    ApplyPosition=true
    ApplyPositionFalloff=false
   )
   &CurlNoiseForce(Noise Strength=400, Pan Noise=false)
   &SolveForcesAndVelocity(Clamp Velocity=false, Limit Acceleration=false)
   &ScaleMeshSize(Scale Factor=&VectorFromFloat(Value=&FloatFromCurve(FloatCurve=DataInterface() {propertyValues {Curve {keys [
          {arriveTangent=2.8779, interpMode=RCIM_Cubic, leaveTangent=2.8779, time=-0.0044, value=0.133}
          {arriveTangent=0.9535, interpMode=RCIM_Cubic, leaveTangent=0.9535, time=0.2999, value=1.0087}
          {arriveTangent=-1.4408, interpMode=RCIM_Cubic, leaveTangent=-1.4408, time=0.8, value=0.9}
          {arriveTangent=-4.5, interpMode=RCIM_Cubic, leaveTangent=-4.5, time=1, value=0}
         ]}}})))
   &DynamicMaterialParameters(
    Index 0 Param 1=&FloatFromCurve(FloatCurve=DataInterface() {propertyValues {Curve {keys [<at 0 0 RCIM_Cubic>,<at 0.5 1 RCIM_Cubic>,<at 1 0 RCIM_Cubic>]}}}, Scale Curve=1.25)
    Index 0 Param 2=&FloatFromCurve(
     FloatCurve=DataInterface() {propertyValues {Curve {keys [<at 0 1 RCIM_Cubic>,<at 0.5 0.1 RCIM_Cubic>,<at 1 1 RCIM_Cubic>]}}}
     Scale Curve=3.5
    )
    Index 0 Param 4=1
    Module.Param0WriteEnabled=true
    Module.Param1WriteEnabled=true
    Module.Param2WriteEnabled=true
    Module.Param3WriteEnabled=true
   ) {enabled=false}
   &UpdateMeshOrientation(Rotation Rate=0.25, Rotation Vector=&RandomVector(Vector Scale=1))
  ]
  rendererClasses [&NiagaraMeshRendererProperties]
  renderers [{rendererClass=&NiagaraMeshRendererProperties, rendererIndex=0}]
 }
 BubbleGuysWithHands {
  emitterUpdateScript [
   &EmitterState()
   &SpawnRate(SpawnRate=2.5, Use Spawn Probability=false)
  ]
  particleSpawnScript [
   &FN_SetSpawnGroup(Number of Spawn Groups=2) {module=FN_SetSpawnGroup001}
   &InitializeParticle(
    Color Mode=::NewEnumerator0
    Lifetime Max=1.4
    Lifetime Min=0.75
    Lifetime Mode=::NewEnumerator1
    Mesh Scale Mode=::NewEnumerator1
    Mesh Uniform Scale Max=2.2
    Mesh Uniform Scale Min=1.8
    UsePositionOffset=false
    Write Material Random=true
   ) {module=InitializeParticle001}
   &SkeletalMeshLocation(
    Bone / Socket Index=$Particles.SpawnGroup
    Bone Sampling Mode=::NewEnumerator4
    Inherit Velocity=false
    MaskSpawnGroup=false
    Sampled Position Offset=&RandomRangeVector(Minimum=<-3 -3 -3>, Maximum=<3 3 3>)
    Skeletal Mesh=$User.Hands
   )
   &InitialMeshOrientation(
    Rotation=&RandomRangeVector(Minimum=<-0.2 0 0>, Maximum=<0.2 0 0>)
    Use Orientation Vector=true
    Use Rotation Vector=true
   )
   &SubUVAnimation(End Frame=3, SubUV Animation Mode:ENiagara_SubUVLookupMode=::NewEnumerator2)
   &AddVelocity(Velocity=&MakeCustomVectorFromBool(
     Bool=$User.bIsFrontEndPreview
     True Vector=&Lerp_Vector(A=<0 0 150>, B=<0 0 400>, Alpha=$System.Dive)
     False Vector=<0 0 0>
    ))
  ]
  particleUpdateScript [
   &ParticleState()
   &InheritSourceMovement(
    Applied Force Scale=&VectorFromFloat(Value=&FloatFromCurve(FloatCurve=DataInterface() {propertyValues {Curve {keys [<at 0 0.7>,<at 1 0>]}}}))
    Applied Position Falloff Distance=700
    Applied Position Scale=&VectorFromFloat(Value=&FloatFromCurve(FloatCurve=DataInterface() {propertyValues {Curve {keys [
          {arriveTangent=-0.2939, interpMode=RCIM_Cubic, leaveTangent=-0.2939, tangentMode=RCTM_User, time=0, value=1}
          {arriveTangent=-1.2984, interpMode=RCIM_Cubic, leaveTangent=-1.2984, tangentMode=RCTM_User, time=1, value=0.5}
         ]}}}))
    ApplyForce=false
    ApplyPosition=true
    ApplyPositionFalloff=false
   )
   &CurlNoiseForce(Noise Strength=200, Pan Noise=false)
   &SolveForcesAndVelocity(Clamp Velocity=false, Limit Acceleration=false)
   &ScaleMeshSize(Scale Factor=&VectorFromFloat(Value=&FloatFromCurve(FloatCurve=DataInterface() {propertyValues {Curve {keys [
          {arriveTangent=2.8779, interpMode=RCIM_Cubic, leaveTangent=2.8779, time=-0.0044, value=0.133}
          {arriveTangent=0.9535, interpMode=RCIM_Cubic, leaveTangent=0.9535, time=0.2999, value=1.0087}
          {arriveTangent=-1.4408, interpMode=RCIM_Cubic, leaveTangent=-1.4408, time=0.8, value=0.9}
          {arriveTangent=-4.5, interpMode=RCIM_Cubic, leaveTangent=-4.5, time=1, value=0}
         ]}}})))
   &UpdateMeshOrientation(Rotation Rate=0.25, Rotation Vector=&RandomVector(Vector Scale=1)) {enabled=false}
   &DynamicMaterialParameters(
    Index 0 Param 1=&RandomRangeFloat(Minimum=0, Maximum=1)
    Index 0 Param 2=1
    Index 0 Param 3=&RandomRangeFloat(Minimum=1.8, Maximum=2.2)
    Module.Param0WriteEnabled=true
    Module.Param1WriteEnabled=true
    Module.Param2WriteEnabled=true
    Module.Param3WriteEnabled=true
   )
  ]
  rendererClasses [&NiagaraMeshRendererProperties]
  renderers [{rendererClass=&NiagaraMeshRendererProperties, rendererIndex=0}]
 }
 Left_hand_tendrils {
  emitterUpdateScript [
   &EmitterState(
    Inactive Response=::NewEnumerator1
    Life Cycle Mode=::NewEnumerator1
    Loop Behavior=::NewEnumerator1
    Loop Duration Mode=::NewEnumerator1
    UseLoopDelay=false
   )
   &SpawnBurst_Instantaneous(Spawn Count=3, Use Spawn Probability=false, UseLoopCountLimit=false)
  ]
  particleSpawnScript [
   &FN_SetSpawnGroup(Number of Spawn Groups=3)
   &InitializeParticle(
    Color Mode=::NewEnumerator0
    Mesh Scale=&Multiply_VectorByFloat(
     Vector=<1 1 1.3>
     Float=&SelectFloatFromArray(Direct Array Index=$Particles.SpawnGroup, Float Selection Array=DataInterface() {propertyValues {FloatData [1.1,1.1,0.9]}})
    )
    Mesh Scale Mode=::NewEnumerator4
    UsePositionOffset=false
    Write Material Random=true
   )
   &InitialMeshOrientation(Use Orientation Vector=true, Use Rotation Vector=true) {enabled=false}
   &CameraOffset(Camera Offset Amount=5)
   /SetVariables(Particles.VisibilityTag=$Particles.SpawnGroup) {isSetParameters=true}
  ]
  particleUpdateScript [
   &ParticleState(Kill Particles When Lifetime Has Elapsed=false, Let Infinitely Lived Particles Die When Emitter Deactivates=true)
   &SkeletalMeshLocation(
    Bone / Socket Index=$Particles.SpawnGroup
    Bone Sampling Mode=::NewEnumerator4
    MaskSpawnGroup=false
    Orientation Sampling=::NewEnumerator1
    Sampled Position Offset=&Lerp_Vector(
     A=&SelectVectorFromArray(Direct Array Index=$Particles.SpawnGroup, Vector Selection Array=DataInterface() {propertyValues {FloatData [<0 0 0>, <0 0 0>, <5 -7 0>]}})
     B=<0 0 0>
     Alpha=$System.Dive
    )
    Skeletal Mesh=$User.Hands
   )
   &SkeletalMeshLocation(
    Bone / Socket Index=$Particles.SpawnGroup
    Bone Sampling Mode=::NewEnumerator4
    MaskSpawnGroup=false
    Orientation Sampling=::NewEnumerator1
    Sampled Position Offset=&Lerp_Vector(
     A=&MaskVectorBySpawnGroup(Masked Vector=<0 4 0>, Passthrough Vector=<0 -4 0>)
     B=&MaskVectorBySpawnGroup(Masked Vector=<0 0 0>, Passthrough Vector=<0 0 0>)
     Alpha=$System.Dive
    )
   ) {enabled=false, module=SkeletalMeshLocation001}
   /SetVariables(Particles.EngineOwnerForwardAxis:Vector3f=$Engine.Owner.SystemXAxis) {isSetParameters=true}
   /SetVariables(Particles.Lean:NiagaraFloat=&RemapRange(
     Input Value To Remap=HlslExpression() {hlslExpression="dot((normalize(Engine.Owner.Velocity)), Particles.EngineOwnerForwardAxis)"}
     Input Min=-0.5
     Input Max=0.5
     Output Min=-1
     Output Max=1
    )) {isSetParameters=true}
   /Game/Effects/Niagara/Modules/Skydiving/FN_DivingAttributes/FN_DivingAttributes_v2(
    Accumulate Dive=0.1
    Accumulate Rest=-0.1
    Bone Index=0
    SkeletalMesh=DataInterface() {propertyValues {
      DefaultMesh=null
      ExcludeBoneName=root
      OptionalComponentClass=null
      PreviewMesh=&M_Med_Soldier_04_Preview
      SkinningMode=SkinOnTheFly
      SoftSourceActor=null
      SourceMode=Default
      UvSetIndex=0
      WholeMeshLOD=-1
      bExcludeBone=false
      bReadDeformedGeometry=true
      bRequireCurrentFrameData=true
      ComponentTags []
      FilteredBones [pelvis]
      FilteredSockets []
      MeshUserParameter {parameter {name=None, typeDefHandle {registeredTypeIndex=1}, varData []}}
      SamplingRegions []
     }}
   ) {enabled=false}
   &CalculateAccurateVelocity() {enabled=false}
   &DynamicMaterialParameters(
    Index 0 Param 1=&RandomRangeFloat(Minimum=0, Maximum=1)
    Index 0 Param 2=&Lerp_Float(A=1, B=2, Alpha=$System.Dive)
    Index 0 Param 3=&RandomRangeFloat(Minimum=1.8, Maximum=2.2)
    Index 1 Param 4=&MakeCustomFloatFromBool(Bool=$User.bIsFrontEndPreview, True Float=0.5, False Float=0.3)
    Module.Param0WriteEnabled=true
    Module.Param0WriteEnabled1=true
    Module.Param1WriteEnabled=true
    Module.Param1WriteEnabled1=true
    Module.Param2WriteEnabled=true
    Module.Param2WriteEnabled1=true
    Module.Param3WriteEnabled=true
    Module.Param3WriteEnabled1=true
    Write Parameter Index 1=true
   )
   &ScaleMeshSize(Scale Factor=&MakeVector(
     X=1
     Y=1
     Z=&Multiply_Float(
      A=&FloatFromCurve(
       FloatCurve=DataInterface() {propertyValues {Curve {keys [<at 0 1>,<at 1 2>]}}}
       CurveIndex=&Divide_Float(A=&VectorLength(Vector=$Engine.Owner.Velocity), B=9000)
       Scale Curve=0.8
      )
      B=&MakeCustomFloatFromBool(Bool=$User.bIsFrontEndPreview, True Float=1.5, False Float=1)
     )
    ))
   /SetVariables(Particles.MeshOrientation=&MakeQuatFromAxisAngle(Axis=$Engine.Owner.SystemYAxis, AngleInDegrees=&Multiply_Float(A=$Particles.Lean, B=-30))) {enabled=false, isSetParameters=true}
   /SetVariables(Particles.MeshOrientation=&LerpQuaternion(
     Quaternion A:Quat4f=$Particles.MeshOrientation
     Quaternion B:Quat4f=&LerpQuaternion(
      Quaternion A:Quat4f=&MakeQuatFromAxisAngle(Axis=$Engine.Owner.SystemXAxis, AngleInDegrees=&Multiply_Float(A=$System.Dive, B=30))
      Quaternion B:Quat4f=<0 0 0 1>
      Lerp Factor:NiagaraFloat=0
      Enable Spherical Linear Interpolation:NiagaraBool=false
     )
     Lerp Factor:NiagaraFloat=$System.Dive
     Enable Spherical Linear Interpolation:NiagaraBool=false
    )) {enabled=false, isSetParameters=true}
   /SetVariables(Particles.PosOffsetForFE:Vector3f=&Add_Vector(
     A=Linked() {linkedVariable {name=Particles.Position, type=NiagaraPosition}}
     B=&RotateVectorByQuaternion(
      Vector To Rotate=&Lerp_Vector(A=<0 -50 0>, B=<0 -300 -100>, Alpha=$System.Dive)
      Quaternion=$Engine.Owner.Rotation
     )
    )) {enabled=false, isSetParameters=true}
   &InterpolateOverTime(
    Rate Of Change=20
    Target Float Value:NiagaraFloat=&MakeFloatFromVector(Vector=Linked() {linkedVariable {name=Particles.Position, type=NiagaraPosition}}, Channel=::NewEnumerator0)
   ) {module=InterpolateOverTime001}
   &InterpolateOverTime(
    Rate Of Change=20
    Target Float Value:NiagaraFloat=&MakeFloatFromVector(Vector=Linked() {linkedVariable {name=Particles.Position, type=NiagaraPosition}}, Channel=::NewEnumerator1)
   ) {module=InterpolateOverTime002}
   &InterpolateOverTime(
    Rate Of Change=45
    Target Float Value:NiagaraFloat=&MakeFloatFromVector(Vector=Linked() {linkedVariable {name=Particles.Position, type=NiagaraPosition}}, Channel=::NewEnumerator2)
   ) {module=InterpolateOverTime003}
   &InterpolateOverTime(
    Rate Of Change=20
    Target Vector Value:Vector3f=&MakeVector(
     X=&MakeFloatFromVector(Vector=Linked() {linkedVariable {name=Particles.Position, type=NiagaraPosition}}, Channel=::NewEnumerator0)
     Y=&MakeFloatFromVector(Vector=Linked() {linkedVariable {name=Particles.Position, type=NiagaraPosition}}, Channel=::NewEnumerator1)
     Z=0
    )
    Variable Type=::NewEnumerator2
   ) {enabled=false}
   /SetVariables(Particles.Interpolate Dynamic=&MakeFloat4FromVectorAndFloat(
     "Vector (RGB)"=&MakeVector(
      X=Linked() {linkedVariable {name=Output.InterpolateOverTime001.Moving Average, type=Quat4f}}
      Y=Linked() {linkedVariable {name=Output.InterpolateOverTime002.Moving Average, type=Quat4f}}
      Z=$Output.InterpolateOverTime003.Moving Average
     )
     "Float (Alpha)"=/Niagara/DynamicInputs/Multiply/ScaleAndBiasFloat(
      Float:NiagaraFloat=0.2
      Scale:NiagaraFloat=&Lerp_Float(A=10, B=1, Alpha=$System.Dive)
      Bias:NiagaraFloat=&Lerp_Float(A=0.15, B=0, Alpha=$System.Dive)
     )
    )) {isSetParameters=true}
   &SolveForcesAndVelocity(Clamp Velocity=false, Limit Acceleration=false)
   /SetVariables(Particles.MeshOrientation=$Engine.Owner.Rotation) {isSetParameters=true}
  ]
  rendererClasses [&NiagaraMeshRendererProperties]
  renderers [
   {rendererClass=&NiagaraMeshRendererProperties, rendererIndex=0}
   {rendererClass=&NiagaraMeshRendererProperties, rendererIndex=1}
   {rendererClass=&NiagaraMeshRendererProperties, rendererIndex=2}
  ]
 }
}
summary {
 systemName=NS_Gumball_Contrail
 emitters [
  {emitterName=BubbleGuys, rendererClasses [&NiagaraMeshRendererProperties]}
  {emitterName=Left_hand_tendrils, rendererClasses [&NiagaraMeshRendererProperties]}
  {emitterName=BubbleGuysWithHands, rendererClasses [&NiagaraMeshRendererProperties]}
 ]
 userVariables [
  User.bIsFrontEnd {type=NiagaraBool}
  User.bIsFrontEndPreview {type=NiagaraBool}
  User.Hands {
   type=NiagaraDataInterfaceSkeletalMesh
   defaultValue {
    struct=NiagaraExt_VariableValue_DataInterface
    value {dataInterface=&self:NiagaraDataInterfaceSkeletalMesh_2}
   }
  }
  User.Interpolate_WPO {type=NiagaraFloat, defaultValue=0, description=""}
  User.SkeletalMesh {
   type=NiagaraDataInterfaceSkeletalMesh
   defaultValue {
    struct=NiagaraExt_VariableValue_DataInterface
    value {dataInterface=&self:NiagaraDataInterfaceSkeletalMesh_6}
   }
  }
 ]
}
